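# Physics asset for the dangling tool props (hammer, intercom, steel tape) on the male medium Builder cosmetic body.
# Intercom and tape each link to their own MSTR anchor body with tight angular limits.
# The hammer hangs in a two-joint chain, limited against dyn_tools_01, which rotates freely on dyn_tools_MSTR.
# Unconstrained capsules on both thighs provide leg collision shapes, while the intercom uses two angled boxes.
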PhysicsAsset /BRCosmetics/Characters/Player/Male/Medium/Bodies/M_MED_Builder/Meshes/M_MED_Builder_dyn_tools
bodies {
 dyn_hammer [{shapeType=Capsule, length=15.75, radius=2, center=<15 0 0>, rotation=<rot 90 0 0>}]
 dyn_intercom [
  {shapeType=Box, extentX=6.75, extentY=5.7274, extentZ=3.2763, center=<9.0515 3 1.5727>, rotation=<rot 0 0 10.0026>}
  {shapeType=Box, extentX=6.75, extentY=5.7274, extentZ=3.2763, center=<9.0517 -3.2 0.9322>, rotation=<rot 0 0 -20.0027>}
 ]
 dyn_intercom_MSTR [{shapeType=Capsule, length=1, radius=1}]
 dyn_steel_tape [{shapeType=Capsule, length=3, radius=3, center=<3 0 0>, rotation=<rot 90 0 0>}]
 dyn_steel_tape_MSTR [{shapeType=Capsule, length=1, radius=1}]
 dyn_tools_01 [{shapeType=Capsule, length=1, radius=1}]
 dyn_tools_MSTR [{shapeType=Capsule, length=1, radius=1}]
 thigh_l [{shapeType=Capsule, length=28.375, radius=7.5, center=<-10.1563 0.1853 -0.3281>, rotation=<rot 88.0317 -2.84 -180>}]
 thigh_r [{shapeType=Capsule, length=19.0261, radius=8.849, center=<8.0617 -0.3978 -0.9975>, rotation=<rot 87.939 -178.3289 -176.3488>}]
}
constraints [
 {bone1Name=dyn_intercom, bone2Name=dyn_intercom_MSTR, swing1LimitDegrees=5, swing2LimitDegrees=7, twistLimitDegrees=2}
 {
  bone1Name=dyn_tools_01
  bone2Name=dyn_tools_MSTR
  swing1LimitDegrees=45
  swing1Motion=Free
  swing2LimitDegrees=45
  swing2Motion=Free
  twistLimitDegrees=45
  twistMotion=Free
 }
 {bone1Name=dyn_hammer, bone2Name=dyn_tools_01, swing1LimitDegrees=45, swing2LimitDegrees=45, twistLimitDegrees=10}
 {bone1Name=dyn_steel_tape, bone2Name=dyn_steel_tape_MSTR, swing1LimitDegrees=10, swing2LimitDegrees=5, twistLimitDegrees=2}
]
