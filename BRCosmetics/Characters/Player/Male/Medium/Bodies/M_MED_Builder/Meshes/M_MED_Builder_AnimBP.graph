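# Anim Blueprint for the male medium Builder body cosmetic, adding deformation and tool-belt motion to a pose copied from the attached parent mesh.
# Twist and sculpter Control Rigs run first, then intercom, tools and steel-tape bones are constrained between pelvis and spine_01.
# Thigh yaw, read each update, drives additive offsets on tools (left) and intercom (right).
# AnimDynamics then swings pliers, screwdrivers and spanners, followed by a spine_02-space RigidBody simulation.
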AnimBlueprint /BRCosmetics/Characters/Player/Male/Medium/Bodies/M_MED_Builder/Meshes/M_MED_Builder_AnimBP
target_skeleton=&/M_MED_Builder_Skeleton
graphs [
 AnimGraph {nodes [
   n1 = Control Rig(Source=Copy Pose From Mesh() {bUseAttachedParent=true}) {
    alphaBoolBlend {blendOption=Linear}
    controlRigAssetReference {blueprintRigClass=&/M_MED_Builder_Deform_CtrlRig_twists.M_MED_Builder_Deform_CtrlRig_twists_C}
   }
   n2 = Control Rig!(
    Source=@n1
    Alpha
    KneeRot_fr_l_114deg:Vector
    KneeTrans_fr_l_114deg:Vector="0.258,-3.591,0.009"
    KneeScale_fr_l_114deg:Vector="1,1,1"
    KneeScale_fr_l_130deg:Vector="1,1,1"
    KneeRot_fr_l_130deg:Vector="0,0,-51.2"
    KneeTrans_fr_l_130deg:Vector="0.258,-3.591,0.009"
   ) -> (Pose) {
    alphaBoolBlend {blendOption=Linear}
    controlRigAssetReference {blueprintRigClass=&/M_MED_Builder_Deform_CtrlRig_sculpters.M_MED_Builder_Deform_CtrlRig_sculpters_C}
   }
   n3 = "Constraint - dyn_intercom_MSTR (PP)"!(ComponentPose=Local To Component(LocalPose=@n2), Alpha=1, ConstraintWeights_0=0.853, ConstraintWeights_1:float=0.147) -> (Pose) {
    alphaBoolBlend {blendOption=Linear}
    boneToModify {boneName=dyn_intercom_MSTR}
    constraintSetup [{targetBone {boneName=pelvis}}, {targetBone {boneName=spine_01}}]
    constraintWeights [1,1]
   }
   n4 = "Constraint - dyn_tools_MSTR (PP)"!(ComponentPose=@n3, Alpha=1, ConstraintWeights_0=0.853, ConstraintWeights_1:float=0.147) -> (Pose) {
    alphaBoolBlend {blendOption=Linear}
    boneToModify {boneName=dyn_tools_MSTR}
    constraintSetup [{targetBone {boneName=pelvis}}, {targetBone {boneName=spine_01}}]
    constraintWeights [1,1]
   }
   n5 = "Constraint - dyn_steel_tape_MSTR (PP)"!(ComponentPose=@n4, Alpha=1, ConstraintWeights_0=0.853, ConstraintWeights_1:float=0.147) -> (Pose) {
    alphaBoolBlend {blendOption=Linear}
    boneToModify {boneName=dyn_steel_tape_MSTR}
    constraintSetup [{targetBone {boneName=pelvis}}, {targetBone {boneName=spine_01}}]
    constraintWeights [1,1]
   }
   n6 = Get thigh_up_l() -> (thigh_up_l:double)
   n7 = "Transform (Modify) Bone - Bone: dyn_tools_MSTR"(ComponentPose=@n5, Alpha=@n6, Translation="0.000,-1.159,0.162", Rotation="0.000,0.000,-3.844", Scale="1,1,1") {
    rotationMode=BMM_Additive
    rotationSpace=BCS_BoneSpace
    translationMode=BMM_Additive
    translationSpace=BCS_BoneSpace
    alphaBoolBlend {blendOption=Linear}
    boneToModify {boneName=dyn_tools_MSTR}
    rotation {roll=-3.844}
    translation {y=-1.159, z=0.162}
   }
   n8 = "Transform (Modify) Bone - Bone: dyn_tools_01"(ComponentPose=@n7, Alpha=@n6, Translation="-0.010,-0.167,-0.045", Rotation="7.828,-6.305,-9.054", Scale="1,1,1") {
    rotationMode=BMM_Additive
    rotationSpace=BCS_BoneSpace
    translationMode=BMM_Additive
    translationSpace=BCS_BoneSpace
    alphaBoolBlend {blendOption=Linear}
    boneToModify {boneName=dyn_tools_01}
    rotation=<rot 7.828 -6.3051 -9.0543>
    translation=<-0.01 -0.167 -0.045>
   }
   n9 = "Transform (Modify) Bone - Bone: dyn_intercom_MSTR"(
    ComponentPose=@n8
    Alpha=Get thigh_up_r() -> (thigh_up_r:double)
    Translation="0.601,4.593,1.238"
    Rotation="-1.657,11.328,20.578"
    Scale="1,1,1"
   ) {
    rotationMode=BMM_Additive
    rotationSpace=BCS_BoneSpace
    translationMode=BMM_Additive
    translationSpace=BCS_BoneSpace
    alphaBoolBlend {blendOption=Linear}
    boneToModify {boneName=dyn_intercom_MSTR}
    rotation=<rot -1.657 11.3282 20.5779>
    translation=<0.601 4.593 1.238>
   }
   n10 = Anim Dynamics - Bone: dyn_pliers(ComponentPose=@n9, Alpha=1, ExternalForce="0,0,0") {
    angularDampingOverride=0.999
    angularSpringConstant=1500
    bAngularSpring=true
    bGravityOverrideInSimSpace=true
    bOverrideAngularDamping=true
    gravityScale=0.35
    simulationSpace=RootRelative
    alphaBoolBlend {blendOption=Linear}
    alphaScaleBiasClamp {scale=0.3}
    boundBone {boneName=dyn_pliers}
    physicsBodyDefinitions [{
      sphereCollisionRadius=0
      boundBone {boneName=dyn_pliers}
      boxExtents=<10 20 20>
      constraintSetup {angularLimitsMax=<5 1 5>, angularLimitsMin=<-5 -1 0>, angularTarget=<0 0 0>, linearAxesMax=<5 0 0>}
      localJointOffset=<-10 5 0>
     }]
   }
   n11 = Anim Dynamics - Bone: dyn_screwdriver_01(ComponentPose=@n10, Alpha=1, ExternalForce="0,0,0") {
    angularDampingOverride=0.999
    angularSpringConstant=1500
    bAngularSpring=true
    bGravityOverrideInSimSpace=true
    bOverrideAngularDamping=true
    gravityScale=0.35
    simulationSpace=RootRelative
    alphaBoolBlend {blendOption=Linear}
    alphaScaleBiasClamp {scale=0.3}
    boundBone {boneName=dyn_screwdriver_01}
    physicsBodyDefinitions [{
      sphereCollisionRadius=0
      boundBone {boneName=dyn_screwdriver_01}
      boxExtents=<10 20 20>
      constraintSetup {angularLimitsMax=<5 5 8>, angularLimitsMin=<-5 0 0>, angularTarget=<0 0 0>, linearAxesMax=<3 0 0>}
      localJointOffset=<-6 -3 0>
     }]
   }
   n12 = Anim Dynamics - Bone: dyn_screwdriver_02(ComponentPose=@n11, Alpha=1, ExternalForce="0,0,0") {
    angularDampingOverride=0.999
    angularSpringConstant=1500
    bAngularSpring=true
    bGravityOverrideInSimSpace=true
    bOverrideAngularDamping=true
    gravityScale=0.35
    simulationSpace=RootRelative
    alphaBoolBlend {blendOption=Linear}
    alphaScaleBiasClamp {scale=0.3}
    boundBone {boneName=dyn_screwdriver_02}
    physicsBodyDefinitions [{
      sphereCollisionRadius=0
      boundBone {boneName=dyn_screwdriver_02}
      boxExtents=<10 20 20>
      constraintSetup {angularLimitsMax=<5 10 10>, angularLimitsMin=<-5 0 -5>, angularTarget=<0 0 0>, linearAxesMax=<2 0 0>}
      localJointOffset=<-3 1 0>
     }]
   }
   n13 = Anim Dynamics - Bone: dyn_spanner_01(ComponentPose=@n12, Alpha=1, ExternalForce="0,0,0") {
    angularDampingOverride=0.999
    angularSpringConstant=1500
    bAngularSpring=true
    bGravityOverrideInSimSpace=true
    bOverrideAngularDamping=true
    gravityScale=0.35
    simulationSpace=RootRelative
    alphaBoolBlend {blendOption=Linear}
    alphaScaleBiasClamp {scale=0.3}
    boundBone {boneName=dyn_spanner_01}
    physicsBodyDefinitions [{
      sphereCollisionRadius=0
      boundBone {boneName=dyn_spanner_01}
      boxExtents=<10 20 20>
      constraintSetup {angularLimitsMax=<2 0 10>, angularLimitsMin=<-2 -2 0>, angularTarget=<0 0 0>, linearAxesMax=<3 0 0>}
      localJointOffset=<-5 -3 0>
     }]
   }
   n14 = Anim Dynamics - Bone: dyn_spanner_02(ComponentPose=@n13, Alpha=1, ExternalForce="0,0,0") {
    angularDampingOverride=0.999
    angularSpringConstant=1500
    bAngularSpring=true
    bGravityOverrideInSimSpace=true
    bOverrideAngularDamping=true
    gravityScale=0.35
    simulationSpace=RootRelative
    alphaBoolBlend {blendOption=Linear}
    alphaScaleBiasClamp {scale=0.3}
    boundBone {boneName=dyn_spanner_02}
    physicsBodyDefinitions [{
      sphereCollisionRadius=0
      boundBone {boneName=dyn_spanner_02}
      boxExtents=<10 20 20>
      constraintSetup {angularLimitsMax=<2 0 0>, angularLimitsMin=<-2 -2 -10>, angularTarget=<0 0 0>, linearAxesMax=<1 0 0>}
      localJointOffset=<-3 2 0>
     }]
   }
   n15 = RigidBody(ComponentPose=@n14) {
    bClampLinearTranslationLimitToRefPose=true
    bOverrideWorldGravity=true
    overridePhysicsAsset=&/M_MED_Builder_dyn_tools
    simulationSpace=BaseBoneSpace
    alphaBoolBlend {blendOption=Linear}
    baseBoneRef {boneName=spine_02}
    componentAppliedLinearAccClamp=<3000 3000 3000>
    componentLinearAccScale=<0.25 0.25 0.25>
    componentLinearVelScale=<0.35 0.35 0.35>
    overrideWorldGravity {z=-980}
   }
   Output Pose(Result=Component To Local(ComponentPose=@n15)) {layerGroup=None}
  ]}
 EventGraph {nodes [
   Event BlueprintUpdateAnimation()
   n1 = BreakTransform(InTransform=GetDeltaTransformFromRefPose(self=GetOwnerSkeletalMeshComponent(), BoneName=thigh_l, BaseName=pelvis))
   > Set thigh_up_l(thigh_up_l:double=MapRangeClamped(Value=BreakRotator(InRot=@n1.Rotation).Yaw, InRangeA=-20, InRangeB=-100)) -> (Output_Get:double)
   n2 = BreakTransform(InTransform=GetDeltaTransformFromRefPose(self=GetOwnerSkeletalMeshComponent(), BoneName=thigh_r, BaseName=pelvis))
   > Set thigh_up_r(thigh_up_r:double=MapRangeClamped(Value=BreakRotator(InRot=@n2.Rotation).Yaw, InRangeA=-20, InRangeB=-100)) -> (Output_Get:double)
  ]}
]
variables [
 thigh_up_l {type=number}
 thigh_up_r {type=number}
]
