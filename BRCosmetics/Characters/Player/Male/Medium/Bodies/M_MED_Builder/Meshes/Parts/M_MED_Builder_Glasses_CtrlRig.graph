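# Control Rig for the glasses of the male medium Builder character, nudging them as the face's cheeks move.
# On forward solve, it reads each inner cheek bone's Z relative to faceAttach and remaps it through a clamped range into a weight.
# Each weight drives an additive local offset and slight tilt on dyn_galss, with opposite roll per side.
# The knot and string bones in the hierarchy are left untouched.
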ControlRigBlueprint /BRCosmetics/Characters/Player/Male/Medium/Bodies/M_MED_Builder/Meshes/Parts/M_MED_Builder_Glasses_CtrlRig
graphs [
 RigVMFunctionLibrary {links [], nodes []}
 RigVMModel {nodes [
   n1 = &RigUnit_BeginExecution()
   /Script/ControlRig.RigUnit_ModifyBoneTransforms(
    BoneToModify=((Bone="dyn_galss",Transform=(Rotation=(X=0,Y=0,Z=0,W=1),Translation=(X=0.3,Y=0,Z=0),Scale3D=(X=1,Y=1,Z=1))))
    Mode=AdditiveLocal
    WeightMaximum=1
    WeightMinimum=0
    ExecutePin=@n1.ExecutePin
    Weight=&RigVMFunction_MathFloatRemap(
     SourceMaximum=5.2
     SourceMinimum=3.7
     TargetMinimum=0
     bClamp=true
     Value=&RigUnit_GetRelativeBoneTransform(Bone=L_cheek_inner, Space=faceAttach).Transform.Translation.Z
    ).Result
    BoneToModify.0.Transform.Rotation=/Script/RigVM.RigVMFunction_MathQuaternionFromRotator(Rotator=(Pitch=1,Yaw=0,Roll=1)) {is=struct RigVMUnitNode}.Result
   ) {is=struct RigVMUnitNode}
   > &RigUnit_ModifyBoneTransforms(
    BoneToModify=((Bone="dyn_galss",Transform=(Rotation=(X=0,Y=0,Z=0,W=1),Translation=(X=0.3,Y=0,Z=0),Scale3D=(X=1,Y=1,Z=1))))
    Mode=AdditiveLocal
    WeightMaximum=1
    WeightMinimum=0
    Weight=&RigVMFunction_MathFloatRemap(
     SourceMaximum=5.2
     SourceMinimum=3.7
     TargetMinimum=0
     bClamp=true
     Value=&RigUnit_GetRelativeBoneTransform(Bone=R_cheek_inner, Space=faceAttach).Transform.Translation.Z
    ).Result
    BoneToModify.0.Transform.Rotation=&RigVMFunction_MathQuaternionFromRotator(Rotator=(Pitch=1,Yaw=0,Roll=-1)).Result
   )
  ]}
]
hierarchy [
 root {loc=<0 0 0>, rot=<0 0 0 1>}
 pelvis {parent=root, loc=<0 -0.7071 93.1191>, rot=<0.0127 0.707 -0.0127 -0.707>}
 spine_01 {parent=pelvis, loc=<6.7494 -0.6388 0>, rot=<0 0 0.034 -0.9994>}
 spine_02 {parent=spine_01, loc=<7.2518 -0.035 0>, rot=<0 0 0.0238 -0.9997>}
 spine_03 {parent=spine_02, loc=<6.0104 0 0>, rot=<0 0 0.0149 -0.9999>}
 spine_04 {parent=spine_03, loc=<8.1024 0 0>, rot=<0 0 -0.0132 -0.9999>}
 spine_05 {parent=spine_04, loc=<10.0146 0 0>, rot=<0 0 -0.0015 -1>}
 neck_01 {parent=spine_05, loc=<11.3138 0.1529 0>, rot=<0 0 -0.2003 -0.9797>}
 neck_02 {parent=neck_01, loc=<4.5627 0 0>, rot=<0 0 0.0504 -0.9987>}
 head {parent=neck_02, loc=<5.3914 0 0>, rot=<0 0 0.115 -0.9934>}
 faceAttach {parent=head, loc=<0.6201 0.9296 0>, rot=<0.0082 -0.7071 -0.0082 -0.7071>, scale=<0.89 0.89 0.89>}
 L_cheek_inner {parent=faceAttach, loc=<4.5352 7.8018 1.6911>, rot=<-0.0619 -0.2477 0.3214 -0.9119>}
 C_nose_bridge {parent=faceAttach, loc=<0 11.5517 2.3424>, rot=<-0.2218 0.005 0.0014 -0.9751>}
 R_cheek_inner {parent=faceAttach, loc=<-4.5352 7.8018 1.6911>, rot=<-0.0619 0.2477 -0.3214 -0.9119>}
 dyn_galss {parent=head, loc=<5.1224 0.1173 0>, rot=<0.4704 0.5279 0.5279 0.4704>}
 dyn_knot_MSTR {parent=head, loc=<-6.0156 5.3058 0.0307>, rot=<0.0029 -0.7071 -0.7071 0.0029>}
 dyn_string_01_r {parent=dyn_knot_MSTR, loc=<0.1393 0.2592 -0.4363>, rot=<-0.0045 -0.0517 -0.0873 -0.9948>}
 dyn_string_02_r {parent=dyn_string_01_r, loc=<2.018 0 0>, rot=<0.0053 0.048 0.0641 -0.9968>}
 dyn_string_01_l {parent=dyn_knot_MSTR, loc=<0.2147 -0.6925 0.6725>, rot=<-0.0021 0.0305 0.0678 -0.9972>}
 dyn_string_02_l {parent=dyn_string_01_l, loc=<1.492 0 0>, rot=<0.0027 -0.0388 -0.0674 -0.997>}
 dyn_string_03_l {parent=dyn_string_02_l, loc=<3.2779 0 0>, rot=<0 -0.0022 -0.0014 -1>}
 dyn_knot_r {parent=dyn_knot_MSTR, loc=<0.5143 0.2773 0.2459>, rot=<0 0 0 -1>}
 dyn_knot_l {parent=dyn_knot_MSTR, loc=<0.0597 -0.34 0.136>, rot=<-0.0105 0.0233 0.4111 -0.9112>}
]
