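# Physics asset for the dynamic knot and strings on the medium male Builder hat, in its no-collar variant.
# Small capsules cover the knot master, both knot sides and the strings (three segments left, two right), while larger capsules cover the upper spine and neck.
# Free-rotation constraints link spine, neck and knot master.
# Knot sides and string chains branch from the master with limited swing and very tight twist.
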PhysicsAsset /BRCosmetics/Characters/Player/Male/Medium/Bodies/M_MED_Builder/Meshes/Parts/M_MED_Builder_Hat_dyn_knot_noCollar
bodies {
 dyn_knot_MSTR [{shapeType=Capsule, length=0.5, radius=0.5}]
 dyn_knot_l [{shapeType=Capsule, length=1, radius=0.75, center=<1 0 0>, rotation=<rot 90 0 0>}]
 dyn_knot_r [{shapeType=Capsule, length=1, radius=0.75, center=<1.5 0 0>, rotation=<rot 90 0 0>}]
 dyn_string_01_l [{shapeType=Capsule, length=0.75, radius=0.75, center=<1 0 0>, rotation=<rot 90 0 0>}]
 dyn_string_01_r [{shapeType=Capsule, length=0.75, radius=0.5, center=<1 0 0>, rotation=<rot 90 0 0>}]
 dyn_string_02_l [{shapeType=Capsule, length=1.5, radius=0.75, center=<1.5 0 0>, rotation=<rot 90 0 0>}]
 dyn_string_02_r [{shapeType=Capsule, length=1.5, radius=0.5, center=<1.5 0 0>, rotation=<rot 90 0 0>}]
 dyn_string_03_l [{shapeType=Capsule, length=1.25, radius=0.75, center=<1.5 0 0>, rotation=<rot 90 0 0>}]
 neck_01 [{shapeType=Capsule, length=1.5, radius=6, rotation=<rot 90 0 0>}]
 neck_02 [{shapeType=Capsule, length=0.75, radius=5.25, rotation=<rot 90 0 0>}]
 spine_05 [{shapeType=Capsule, length=13.5, radius=11, center=<-0.1045 -1.0374 -0.0001>}]
}
constraints [
 {
  bone1Name=neck_01
  bone2Name=spine_05
  swing1LimitDegrees=45
  swing1Motion=Free
  swing2LimitDegrees=45
  swing2Motion=Free
  twistLimitDegrees=45
  twistMotion=Free
 }
 {
  bone1Name=neck_02
  bone2Name=neck_01
  swing1LimitDegrees=45
  swing1Motion=Free
  swing2LimitDegrees=45
  swing2Motion=Free
  twistLimitDegrees=45
  twistMotion=Free
 }
 {
  bone1Name=dyn_knot_MSTR
  bone2Name=neck_02
  swing1LimitDegrees=45
  swing1Motion=Free
  swing2LimitDegrees=45
  swing2Motion=Free
  twistLimitDegrees=45
  twistMotion=Free
 }
 {bone1Name=dyn_string_01_r, bone2Name=dyn_knot_MSTR, swing1LimitDegrees=30, swing2LimitDegrees=30, twistLimitDegrees=2}
 {bone1Name=dyn_string_02_r, bone2Name=dyn_string_01_r, swing1LimitDegrees=30, swing2LimitDegrees=30, twistLimitDegrees=2}
 {bone1Name=dyn_string_01_l, bone2Name=dyn_knot_MSTR, swing1LimitDegrees=30, swing2LimitDegrees=30, twistLimitDegrees=2}
 {bone1Name=dyn_string_02_l, bone2Name=dyn_string_01_l, swing1LimitDegrees=30, swing2LimitDegrees=30, twistLimitDegrees=2}
 {bone1Name=dyn_string_03_l, bone2Name=dyn_string_02_l, swing1LimitDegrees=30, swing2LimitDegrees=30, twistLimitDegrees=2}
 {bone1Name=dyn_knot_r, bone2Name=dyn_knot_MSTR, swing1LimitDegrees=30, swing2LimitDegrees=30, twistLimitDegrees=2}
 {bone1Name=dyn_knot_l, bone2Name=dyn_knot_MSTR, swing1LimitDegrees=15, swing2LimitDegrees=15, twistLimitDegrees=2}
]
