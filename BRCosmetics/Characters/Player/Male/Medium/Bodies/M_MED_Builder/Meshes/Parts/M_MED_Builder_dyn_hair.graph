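# Physics asset for the dynamic hair on the M_MED_Builder male medium player character, defining bodies for secondary hair motion.
# Four small capsules cover the dyn_hair bones, and a zero-length capsule forms a spherical body on the head.
# Each hair bone is constrained directly to the head rather than chained, with narrow swing and twist limits that are tightest on dyn_hair_04.
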PhysicsAsset /BRCosmetics/Characters/Player/Male/Medium/Bodies/M_MED_Builder/Meshes/Parts/M_MED_Builder_dyn_hair
bodies {
 dyn_hair_01 [{shapeType=Capsule, length=2.4626, radius=1.5374, center=<3.2619 0 0.2619>, rotation=<rot 45 0 0>}]
 dyn_hair_02 [{shapeType=Capsule, length=2.4626, radius=1.5374, center=<5.0348 0.0003 -1.2567>, rotation=<rot 45 0 0>}]
 dyn_hair_03 [{shapeType=Capsule, length=2.296, radius=1.704, center=<2.9574 0.0001 -1.0595>, rotation=<rot 64.9991 -0.0001 0.0004>}]
 dyn_hair_04 [{shapeType=Capsule, length=1.0903, radius=1.6438, center=<1.3338 0.6806 -0.3903>, rotation=<rot 72.4652 -3.8795 -28.6711>}]
 head [{shapeType=Capsule, length=0, radius=10.4599, center=<2.8873 3.1003 0>, rotation=<rot 90 0 0>}]
}
constraints [
 {bone1Name=dyn_hair_01, bone2Name=head, swing1LimitDegrees=30, swing2LimitDegrees=15, twistLimitDegrees=5}
 {bone1Name=dyn_hair_02, bone2Name=head, swing1LimitDegrees=20, swing2LimitDegrees=15, twistLimitDegrees=5}
 {bone1Name=dyn_hair_03, bone2Name=head, swing1LimitDegrees=20, swing2LimitDegrees=15, twistLimitDegrees=5}
 {bone1Name=dyn_hair_04, bone2Name=head, swing1LimitDegrees=15, swing2LimitDegrees=10, twistLimitDegrees=2}
]
