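# Animation Blueprint for the Builder male medium character's hat part, adding secondary motion on top of the parent mesh's pose.
# It copies pose and curves from the attached parent, then runs the Glasses Control Rig.
# A constraint blends the dyn_knot_MSTR bone mostly toward the jaw over the head, feeding a head-relative RigidBody simulation using the no-collar knot physics asset.
# A neck-based hair RigidBody node is unconnected, and the EventGraph holds only stubs.
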AnimBlueprint /BRCosmetics/Characters/Player/Male/Medium/Bodies/M_MED_Builder/Meshes/Parts/M_MED_Builder_Hat_AnimBP
target_skeleton=&/M_MED_Builder_Skeleton
graphs [
 AnimGraph {nodes [
   RigidBody() {
    bClampLinearTranslationLimitToRefPose=true
    bOverrideWorldGravity=true
    overridePhysicsAsset=&/M_MED_Builder_dyn_hair
    simulationSpace=BaseBoneSpace
    alphaBoolBlend {blendOption=Linear}
    baseBoneRef {boneName=neck_01}
    componentAppliedLinearAccClamp=<100 100 100>
    componentLinearAccScale=<0.2 0.2 0.2>
    componentLinearVelScale=<0.2 0.2 0.2>
    overrideWorldGravity {z=-100}
   }
   n1 = Control Rig(Source=Copy Pose From Mesh() {bCopyCurves=true, bUseAttachedParent=true}) {
    alphaBoolBlend {blendOption=Linear}
    controlRigAssetReference {blueprintRigClass=&/M_MED_Builder_Glasses_CtrlRig.M_MED_Builder_Glasses_CtrlRig_C}
   }
   n2 = "Constraint - dyn_knot_MSTR (PP)"!(ComponentPose=Local To Component(LocalPose=@n1), Alpha=1, ConstraintWeights_0=0.7842, ConstraintWeights_1:float=0.2158) -> (Pose) {
    alphaBoolBlend {blendOption=Linear}
    boneToModify {boneName=dyn_knot_MSTR}
    constraintSetup [{targetBone {boneName=C_jaw}}, {targetBone {boneName=head}}]
    constraintWeights [1,1]
   }
   n3 = RigidBody(ComponentPose=@n2) {
    bClampLinearTranslationLimitToRefPose=true
    bOverrideWorldGravity=true
    overridePhysicsAsset=&/M_MED_Builder_Hat_dyn_knot_noCollar
    simulationSpace=BaseBoneSpace
    alphaBoolBlend {blendOption=Linear}
    baseBoneRef {boneName=head}
    componentAppliedLinearAccClamp=<3000 3000 3000>
    componentLinearAccScale=<0.3 0.3 0.3>
    componentLinearVelScale=<0.3 0.3 0.3>
    overrideWorldGravity {z=-980}
   }
   Output Pose(Result=Component To Local(ComponentPose=@n3)) {layerGroup=None}
  ]}
 EventGraph {nodes [
   TryGetPawnOwner()
   Event BlueprintUpdateAnimation()
  ]}
]
variables []
